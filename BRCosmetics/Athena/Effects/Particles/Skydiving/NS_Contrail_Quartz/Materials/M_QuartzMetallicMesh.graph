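# Lit metallic material for Niagara mesh particles in the Quartz skydiving contrail, adding rainbow thin-film sheen and a sine wiggle.
# The reflection vector samples a thin-film color texture twice, tinting base color and feeding emissive, both scaled by the FrontendBoost dynamic parameter.
# A Z-axis sine offset uses time, UV and red vertex color.
# The Diving dynamic parameter blends its frequency, amplitude and UV phase between low and high settings.
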Material /BRCosmetics/Athena/Effects/Particles/Skydiving/NS_Contrail_Quartz/Materials/M_QuartzMetallicMesh
graph {
 properties {bUsedWithNiagaraMeshParticles=true, pixelDepthOffsetMode=PDOM_Legacy, usedShadingModels=Default Lit}
 nodes [
  Constant() {r=2}
  Constant() {r=0.2}
  Constant() {r=0.2}
  Constant() {r=1}
  n1 = DynamicParameter() {parameterIndex=3, defaultValue=<rgba 0 1 1 1>, paramNames [Diving, FrontendBoost, Param3, Param4]}
  n2 = ReflectionVectorWS()
  n3 = ComponentMask(None=@n2) {g=true, r=true}
  n4 = ScalarParameter() {defaultValue=0.25, parameterName=Reflection Scale}
  n5 = ScalarParameter() {defaultValue=0.532, parameterName=ReflectionOffset}
  n6 = ComponentMask(None=@n2) {b=true}
  n7 = LinearInterpolate(
   A=TextureSample(UVs=MaterialFunctionCall!(UV=@n3, Tile U=@n4, Tile V=@n4, Pan U, Pan V, Offset U=@n5, Offset V, Time) {materialFunction=/Game/Effects/Fort_Effects/Materials/_Functions/MF_PanTileOffset}.UV's) {texture=/Game/Gadgets/Assets/VinderTech_GliderChute/Glider_Quartz/Textures/T_ThinFilm_Quartz_Color}.RGB
   B=TextureSample(UVs=MaterialFunctionCall(UV=@n6, Tile U=@n4, Tile V=@n4, Offset U=@n5) {materialFunction=&MF_PanTileOffset}.UV's) {texture=&T_ThinFilm_Quartz_Color}.RGB
  ) {constAlpha=0.75}
  n8 = Power(Base=@n7) {constExponent=1.5}
  n9 = ScalarParameter() {defaultValue=0.25, parameterName=EmissiveContribution}
  MaterialFunctionCall(Emissive=Multiply(A=@n1.FrontendBoost, B=Multiply(A=@n8, B=@n9))) {materialFunction=&MF_FX_EmissiveAdjustments}
  Multiply(
   A=LinearInterpolate(
    A=VectorParameter() {parameterName=Base Color, defaultValue=<rgba 1 1 1 1>}.RGB
    B=@n7
    Alpha=ScalarParameter() {defaultValue=0.25, parameterName=RainbowStrength}
   )
   B=@n1.FrontendBoost
  )
  n10 = TextureCoordinate()
  n11 = ScalarParameter() {defaultValue=0.5, parameterName=WiggleStrengthLow}
  n12 = ScalarParameter() {defaultValue=1, parameterName=WiggleStrengthHigh}
  n13 = DynamicParameter() {defaultValue=<rgba 0 1 1 1>, paramNames [Diving,Param2,Param3,Param4]}
  n14 = VertexColor()
  n15 = Multiply(A=Multiply(A=ComponentMask(None=@n10) {g=true}, B=LinearInterpolate(A=@n11, B=@n12, Alpha=@n13.Diving)), B=Floor(None=OneMinus(None=@n14.R)))
  n16 = Time()
  n17 = ScalarParameter() {defaultValue=0.2, parameterName=SineFreqLow}
  n18 = Sine(None=Add(A=Add(A=@n15, B=@n14.R), B=Multiply(A=@n16, B=@n17)))
  n19 = TextureCoordinate()
  n20 = ScalarParameter() {defaultValue=0.5, parameterName=WiggleStrengthLow}
  n21 = ScalarParameter() {defaultValue=1, parameterName=WiggleStrengthHigh}
  n22 = DynamicParameter() {defaultValue=<rgba 0 1 1 1>, paramNames [Diving,Param2,Param3,Param4]}
  n23 = VertexColor()
  n24 = Multiply(A=Multiply(A=ComponentMask(None=@n19) {g=true}, B=LinearInterpolate(A=@n20, B=@n21, Alpha=@n22.Diving)), B=Floor(None=OneMinus(None=@n23.R)))
  n25 = Time()
  n26 = ScalarParameter() {defaultValue=1, parameterName=SineFreqHigh}
  n27 = Sine(None=Add(A=Add(A=@n24, B=@n23.R), B=Multiply(A=@n25, B=@n26)))
  n28 = DynamicParameter() {defaultValue=<rgba 0 1 1 1>, paramNames [Diving,Param2,Param3,Param4]}
  Multiply(
   A=Multiply(A=LinearInterpolate(A=@n18, B=@n27, Alpha=@n28.Diving), B=Constant3Vector() {constant=<rgba 0 0 1 0>})
   B=LinearInterpolate(
    A=ScalarParameter() {defaultValue=8, parameterName=SineStrengthLow}
    B=ScalarParameter() {defaultValue=14, parameterName=SineStrengthHigh}
    Alpha=DynamicParameter() {defaultValue=<rgba 0 1 1 1>, paramNames [Diving,Param2,Param3,Param4]}.Diving
   )
  )
 ]
}
