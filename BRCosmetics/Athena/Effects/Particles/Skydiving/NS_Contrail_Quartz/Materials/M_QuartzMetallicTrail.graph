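# Masked, lit material for Niagara mesh particles in the Quartz skydiving contrail, giving the trail a metallic rainbow look.
# A thin-film texture sampled by world reflection vector tints base color and emissive, both scaled by a FrontendBoost dynamic parameter.
# Vertex offsets sum a UV-based shrink wipe and puff along normals, velocity-driven lag, and a sine/cosine waggle that blends to diving settings.
# Particle dynamic parameters drive the wipe, puff, lag and waggle timing.
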Material /BRCosmetics/Athena/Effects/Particles/Skydiving/NS_Contrail_Quartz/Materials/M_QuartzMetallicTrail
graph {
 properties {bUsedWithNiagaraMeshParticles=true, blendMode=BLEND_Masked, pixelDepthOffsetMode=PDOM_Legacy, usedShadingModels=Default Lit}
 nodes [
  n1 = OneMinus(None=ComponentMask(None=TextureCoordinate()) {g=true})
  n2 = DynamicParameter() {defaultValue=<rgba 1 0 1 1>, paramNames [ShrinkWipeOn,ForceShrink,VelX,VelY]}
  n3 = Saturate(None=Multiply(A=Subtract(A=@n1, B=@n2.ShrinkWipeOn), B=ScalarParameter() {defaultValue=5, parameterName=ShrinkFalloff}))
  n4 = DynamicParameter() {defaultValue=<rgba 1 0 1 1>, paramNames [ShrinkWipeOn,ForceShrink,VelX,VelY]}
  n5 = ScalarParameter() {defaultValue=-5.5, parameterName=ShrinkMax}
  n6 = MaterialFunctionCall(Vector 3=Transform(None=Constant3Vector() {constant=<rgba 1 0 0 0>}) {transformSourceType=TRANSFORMSOURCE_Instance}) {materialFunction=/Engine/Functions/Engine_MaterialFunctions02/Utility/VectorLength}
  n7 = Multiply(
   A=Multiply(A=VertexColor().R, B=VertexNormalWS())
   B=Multiply(A=LinearInterpolate(A=@n3, Alpha=@n4.ForceShrink), B=Multiply(A=@n5, B=@n6.V3 Length))
  )
  n8 = NamedRerouteUsage()
  n9 = Multiply(
   A=DynamicParameter() {parameterIndex=1, defaultValue=<rgba 0 1 1 1>, paramNames [Puff,UnMirror,TimingOffset,Diving]}.Puff
   B=VertexNormalWS()
  )
  n10 = ComponentMask(None=TextureCoordinate()) {g=true}
  n11 = Add(
   A=LinearInterpolate(Alpha=DynamicParameter() {defaultValue=<rgba 1 0 1 1>, paramNames [ShrinkWipeOn,ForceShrink,VelX,VelY]}.ShrinkWipeOn) {constA=-1, constB=0}
   B=@n10
  )
  n12 = Multiply(A=Multiply(A=Saturate(None=@n11), B=OneMinus(None=@n10))) {constB=4}
  Add(A=Add(A=Add(A=@n7, B=@n8), B=Multiply(A=@n9, B=@n12)), B=NamedRerouteUsage())
  Constant() {r=1}
  Constant() {r=0.2}
  Constant() {r=1}
  Constant() {r=0.2}
  DynamicParameter() {parameterIndex=1, defaultValue=<rgba 0 1 1 1>, paramNames [Puff,UnMirror,TimingOffset,Diving]}
  n13 = DynamicParameter() {parameterIndex=3, defaultValue=<rgba 1 0 1 1>, paramNames [FrontendBoost, ForceShrink, VelX, VelY]}
  n14 = ReflectionVectorWS()
  n15 = ComponentMask(None=@n14) {g=true, r=true}
  n16 = ScalarParameter() {defaultValue=0.25, parameterName=Reflection Scale}
  n17 = ScalarParameter() {defaultValue=0.55, parameterName=ReflectionOffset}
  n18 = ComponentMask(None=@n14) {b=true}
  n19 = LinearInterpolate(
   A=TextureSample(UVs=MaterialFunctionCall!(UV=@n15, Tile U=@n16, Tile V=@n16, Pan U, Pan V, Offset U=@n17, Offset V, Time) {materialFunction=/Game/Effects/Fort_Effects/Materials/_Functions/MF_PanTileOffset}.UV's) {texture=/Game/Gadgets/Assets/VinderTech_GliderChute/Glider_Quartz/Textures/T_ThinFilm_Quartz_Color}.RGB
   B=TextureSample(UVs=MaterialFunctionCall(UV=@n18, Tile U=@n16, Tile V=@n16, Offset U=@n17) {materialFunction=&MF_PanTileOffset}.UV's) {texture=&T_ThinFilm_Quartz_Color}.RGB
  ) {constAlpha=0.75}
  n20 = Power(Base=@n19, Exp=ScalarParameter() {defaultValue=1, parameterName=Reflection Exp}) {constExponent=1.5}
  n21 = ScalarParameter() {defaultValue=0.35, parameterName=EmissiveContribution}
  MaterialFunctionCall(Emissive=Multiply(A=@n13.FrontendBoost, B=Multiply(A=@n20, B=@n21))) {materialFunction=&MF_FX_EmissiveAdjustments}
  Multiply(
   A=LinearInterpolate(
    A=VectorParameter() {parameterName=Base Color, defaultValue=<rgba 1 1 1 1>}.RGB
    B=@n19
    Alpha=ScalarParameter() {defaultValue=0.5, parameterName=RainbowStrength}
   )
   B=@n13.FrontendBoost
  )
  n22 = DynamicParameter() {defaultValue=<rgba 1 0 1 1>, paramNames [ShrinkWipeOn,ForceShrink,VelX,VelY]}
  n23 = Constant()
  n24 = Clamp(None=Divide(A=MaterialFunctionCall(R=@n22.VelX, G=@n22.VelX, B=@n23) {materialFunction=&AppendMany}.RGB) {constB=-960}) {minDefault=-1}
  n25 = ScalarParameter() {defaultValue=150, parameterName=LagStrength}
  n26 = ComponentMask(None=TextureCoordinate()) {g=true}
  n27 = Saturate(None=Power(Base=OneMinus(None=@n26)) {constExponent=3})
  n28 = ScalarParameter() {defaultValue=1, parameterName=LagPowerA}
  NamedRerouteDeclaration(None=Multiply(A=Multiply(A=@n24, B=@n25), B=Multiply(A=@n27, B=@n28))) {name=VertexLag, nodeColor=<rgba 1 0.3281 0 1>}
  n29 = DynamicParameter() {parameterIndex=1, defaultValue=<rgba 0 1 1 1>, paramNames [Puff,UnMirror,TimingOffset,Diving]}
  n30 = Constant3Vector() {constant=<rgba 0 1 0 0>}
  n31 = Time()
  n32 = Multiply(A=ComponentMask(None=TextureCoordinate()) {g=true}, B=ScalarParameter() {parameterName=Frequency})
  n33 = DynamicParameter() {parameterIndex=1, defaultValue=<rgba 0 1 1 1>, paramNames [Puff,UnMirror,TimingOffset,Diving]}
  n34 = Multiply(A=Add(A=Add(A=@n31, B=@n32), B=@n33.TimingOffset), B=ScalarParameter() {parameterName=SineSpeed})
  n35 = Cosine(None=@n34)
  n36 = Time()
  n37 = Multiply(A=ComponentMask(None=TextureCoordinate()) {g=true}, B=ScalarParameter() {parameterName=DiveFrequency})
  n38 = Multiply(
   A=Add(
    A=Add(A=@n36, B=@n37)
    B=DynamicParameter() {parameterIndex=1, defaultValue=<rgba 0 1 1 1>, paramNames [Puff,UnMirror,TimingOffset,Diving]}.TimingOffset
   )
   B=ScalarParameter() {parameterName=DiveSineSpeed}
  )
  n39 = Cosine(None=@n38)
  n40 = Sine(None=@n34)
  n41 = Sine(None=@n38)
  n42 = Constant3Vector() {constant=<rgba 1 0 0 0>}
  n43 = Add(
   A=Multiply(A=Multiply(A=@n29.UnMirror, B=@n30), B=LinearInterpolate(A=@n35, B=@n39, Alpha=@n33.Diving))
   B=Multiply(A=LinearInterpolate(A=@n40, B=@n41, Alpha=@n33.Diving), B=Multiply(A=@n42))
  )
  n44 = Multiply(A=ScalarParameter() {parameterName=SineAmount})
  n45 = ComponentMask(None=TextureCoordinate()) {g=true}
  n46 = Multiply(A=@n45, B=OneMinus(None=@n45)) {constB=4}
  NamedRerouteDeclaration(None=Multiply(A=Multiply(A=@n43, B=@n44), B=Multiply(A=@n46) {constB=4})) {name=Waggle, nodeColor=<rgba 1 0.5391 0 1>}
  OneMinus(None=@n3)
 ]
}
